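# Material function for the Brutal Burglar body skin that overlays a configurable "SimpleMaterial" metallic layer, named as gold, onto incoming material attributes.
# Base color combines desaturated original color, GMap green and a tint or texture, whose alpha can drive roughness.
# SkinFX or super-levelling channel masks control the blend, with an edge channel retaining original emissive.
# Static switches select full override, shading-model override, or base color and metallic only.
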MaterialFunction /BRCosmetics/Characters/Player/Female/Medium/Bodies/F_MED_BrutalBurglar/Materials/MF_BrutalBurglar_GoldMaterial
graph {nodes [
  n1 = GetMaterialAttributes(None=FunctionInput() {inputName=Material, inputType=FunctionInput_MaterialAttributes}) {attributeGetTypes [#1, #2, #3]}
  n2 = Desaturation(None=@n1.BaseColor, Fraction=Constant() {r=1})
  n3 = ScalarParameter() {defaultValue=0.5, group=SimpleMaterial, parameterName=SM_BaseColorInfluence}
  n4 = Saturate(None=Add(A=Multiply(A=@n2) {constB=5}, B=OneMinus(None=@n3)) {constB=0.5})
  n5 = Add(A=ComponentMask(None=FunctionInput() {inputName=GMap}) {g=true}) {constB=0.275}
  n6 = TextureSampleParameter2D() {
   group=SimpleMaterial
   parameterName="SM_BaseColor,Roughness(RGB,A)"
   texture=/Game/Weapons/WeaponSkins/Wraps/Textures/CatBurglarGameplay/T_Wrap_CatBurglarGameplay_D
  }
  n7 = VectorParameter() {group=SimpleMaterial, parameterName=SM_BaseColor, defaultValue=<rgba 0 0 0 1>}
  n8 = Multiply(
   A=StaticSwitchParameter(True=OneMinus(None=@n4), False=@n4) {group=SimpleMaterial, parameterName=SM_InvertBaseColorInfluence}
   B=Multiply(A=Saturate(None=@n5), B=StaticSwitchParameter(True=@n6.RGB, False=@n7.RGB) {group=SimpleMaterial, parameterName=SM_UseTexture})
  )
  n9 = ScalarParameter() {defaultValue=1, group=SimpleMaterial, parameterName=SM_Metallic}
  n10 = StaticSwitchParameter(
   True=LinearInterpolate(
    A=ScalarParameter() {group=SimpleMaterial, parameterName=SM_MinRoughness}
    B=ScalarParameter() {defaultValue=1, group=SimpleMaterial, parameterName=SM_MaxRoughness}
    Alpha=@n6.A
   )
   False=ScalarParameter() {defaultValue=0.1, group=SimpleMaterial, parameterName=SM_Roughness}
  ) {group=SimpleMaterial, parameterName=SM_UseTexture}
  n11 = ScalarParameter() {defaultValue=2, group=SimpleMaterial, parameterName=SM_Specular}
  n12 = Constant3Vector()
  n13 = FunctionInput(Preview=Constant()) {bUsePreviewValueAsDefault=true, inputName=SkinFXMask, inputType=FunctionInput_Vector4}
  n14 = ChannelMaskParameter(None=@n13) {group=SimpleMaterial, parameterName=SM_MaskChannel}
  n15 = LinearInterpolate(A=@n14.R, B=OneMinus(None=@n14.R), Alpha=ScalarParameter() {group=SimpleMaterial, parameterName=SM_InvertChannelMask})
  n16 = SmoothStep(
   Min=ScalarParameter() {group=SimpleMaterial, parameterName=SM_SuperLevelMaskFX_SStepMin, sortPriority=115}
   Max=ScalarParameter() {defaultValue=1, group=SimpleMaterial, parameterName=SM_SuperLevelMaskFX_SStepMax, sortPriority=120}
   Value=@n15
  )
  n17 = FunctionInput(Preview=Constant()) {bUsePreviewValueAsDefault=true, inputName=SuperLevellingMask}
  n18 = ChannelMaskParameter(None=@n17) {group=SimpleMaterial, maskChannel=Green, parameterName=SM_SuperLevellingEdgeChannel, defaultValue=<rgba 0 1 0 0>}
  n19 = LinearInterpolate(
   A=@n12
   B=@n1.EmissiveColor
   Alpha=StaticSwitchParameter(True=Multiply(A=@n16, B=@n18.R), False=@n18.R) {group=SimpleMaterial, parameterName=SM_SuperLevelingEdge_UseSkinFXMask, sortPriority=110}
  )
  n20 = SetMaterialAttributes!(
   MaterialAttributes=@n1.MaterialAttributes
   BaseColor=@n8
   Metallic=@n9
   Roughness=@n10
   Normal
   Specular=@n11
   SubsurfaceColor=@n12
   EmissiveColor=@n19
   ShadingModel=ShadingModel()
  ) {attributeSetTypes [
    #3
    #4
    #5
    #1
    #6
    #7
    #2
    #8
   ]}
  n21 = LinearInterpolate(
   A=@n1.BaseColor
   B=Multiply(A=@n8, B=VectorParameter() {group=SimpleMaterial, parameterName=SM_DarkenBaseColor, defaultValue=<rgba 1 1 1 0>}.RGB)
   Alpha=ChannelMaskParameter(None=@n13) {group=SimpleMaterial, parameterName=SM_BaseColorMaskChannel_FromFX}.R
  )
  n22 = SetMaterialAttributes!(
   MaterialAttributes=@n1.MaterialAttributes
   BaseColor=@n8
   Metallic=@n9
   Roughness=@n10
   Normal
   Specular=@n11
   SubsurfaceColor=@n12
   EmissiveColor=@n19
  ) {attributeSetTypes [
    #3
    #4
    #5
    #1
    #6
    #7
    #2
   ]}
  n23 = StaticSwitchParameter(
   True=@n20
   False=StaticSwitchParameter(
    True=SetMaterialAttributes!(MaterialAttributes=@n1.MaterialAttributes, BaseColor=@n21, Metallic=@n9) {attributeSetTypes [#3, #4]}
    False=@n22
   ) {group=SimpleMaterial, parameterName="SM_OverrideOnlyBaseColor&&Metallic", sortPriority=151}
  ) {group=SimpleMaterial, parameterName=SM_OverrideShadingModel, sortPriority=150}
  n24 = ChannelMaskParameter(None=@n17) {group=SimpleMaterial, parameterName=SM_SuperLEvellingMaskChannel}
  n25 = StaticSwitchParameter(True=Multiply(A=@n16, B=@n24.R), False=@n24.R) {group=SimpleMaterial, parameterName=SM_SuperLevelingMask_UseSkinFXMask, sortPriority=100}
  FunctionOutput(None=StaticSwitchParameter(
    True=BlendMaterialAttributes(A=@n1.MaterialAttributes, B=@n23, Alpha=@n25) {vertexAttributeBlendType=UseA}
    False=BlendMaterialAttributes(A=@n1.MaterialAttributes, B=@n23, Alpha=@n15)
   ) {group=SimpleMaterial, parameterName=SM_UseSuperlevellingMask})
  Multiply()
 ]}
