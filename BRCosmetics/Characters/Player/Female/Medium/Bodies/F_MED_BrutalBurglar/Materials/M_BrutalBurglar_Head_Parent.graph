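# Parent character material for the Brutal Burglar head, a skeletal mesh material with morph target support and a hero skin subsurface profile.
# It chains material-attribute functions, starting with the shared character PreFX base, whose pre-skinned local position feeds a super-levelling mask.
# That mask and the base Gmap drive a Brutal Burglar gold material layer.
# The result passes through shared character PostFX, then a cinematic PostFX function.
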Material /BRCosmetics/Characters/Player/Female/Medium/Bodies/F_MED_BrutalBurglar/Materials/M_BrutalBurglar_Head_Parent
graph {
 properties {
  bAutomaticallySetUsageInEditor=false
  bUseMaterialAttributes=true
  bUsedWithMeshDeformer=true
  bUsedWithMorphTargets=true
  bUsedWithSkeletalMesh=true
  pixelDepthOffsetMode=PDOM_Legacy
  refractionMethod=RM_IndexOfRefraction
  shadingModel=MSM_FromMaterialExpression
  subsurfaceProfile=&SS_HeroSkin_02
  usedShadingModels=Default Lit | Subsurface Profile
 }
 nodes [
  n1 = MaterialFunctionCall() {materialFunction=&BaseFN_CharacterFunction_PreFX}
  n2 = MaterialFunctionCall!(Local Normal, Local Position=@n1.Pre skinned Local Position, Material=@n1.Pre skinned Local Position) {materialFunction=/Game/Packages/Fortress_SharedMaterials/Base_Material_Functions/MF_SuperLevellingMask}
  n3 = MaterialFunctionCall(
   AO=@n1.AO
   Distance Blend=@n1.AO
   Fresnel=@n1.AO
   Gmap=@n1.AO
   Pre Skinned Local Position=@n1.AO
   Material=MaterialFunctionCall!(SuperLevellingMask=@n2.SuperLevelling_Mask, SkinFXMask, GMap=@n1.Gmap, Material=@n2.SuperLevelling_Mask) {materialFunction=&/MF_BrutalBurglar_GoldMaterial}.Result
  ) {materialFunction=&BaseFN_CharacterFunction_PostFX}
  MaterialFunctionCall(In=@n3.Result) {materialFunction=&CINE_MF_CharPostFX}
 ]
}
