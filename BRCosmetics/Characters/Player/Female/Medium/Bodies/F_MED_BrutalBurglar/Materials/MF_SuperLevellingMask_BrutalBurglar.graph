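# Material function for the Brutal Burglar outfit's super-levelling effect, outputting a reveal/edge mask and material attributes with optional glowing-edge emissive.
# Level, animated from the previous one using Blueprint-supplied start time, is normalized, optionally quantized and remapped through a curve atlas.
# That curve drives a sphere mask's radius and three-point center path in pre-skinned space, broken up by projected noise.
# Edge glow follows the level-up animation or stays on, optionally pulsing, with optional region masking.
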MaterialFunction /BRCosmetics/Characters/Player/Female/Medium/Bodies/F_MED_BrutalBurglar/Materials/MF_SuperLevellingMask_BrutalBurglar
graph {nodes [
  n1 = PreSkinnedPosition()
  n2 = VectorParameter() {group=SuperLevelling_Mask, parameterName=SL_Backpack_PreskinnedPositionOffset, defaultValue=<rgba 0 0 0 1>}
  n3 = Add(A=FunctionInput(Preview=VertexInterpolator(VS=@n1).PS) {bUsePreviewValueAsDefault=true, inputName=Local Position}, B=@n2.RGB)
  n4 = ScalarParameter() {defaultValue=100, group=SuperLevelling_Mask, parameterName=SL_CurrentLevel, sortPriority=3}
  n5 = Time()
  n6 = ScalarParameter() {defaultValue=-99999, group=SuperLevelling_Mask, parameterName=SL_GameTimeFromBP, sortPriority=100}
  n7 = ScalarParameter() {defaultValue=2, group=SuperLevelling_Mask, parameterName=SL_Animation_Duration, sortPriority=101}
  n8 = Saturate(None=Divide(A=Subtract(A=@n5, B=@n6), B=Add(A=@n7) {constB=0}))
  n9 = LinearInterpolate(A=Subtract(A=@n4), B=@n4, Alpha=@n8)
  n10 = ScalarParameter() {defaultValue=100, group=SuperLevelling_Mask, parameterName=SL_StartAtLevel, sortPriority=3}
  n11 = ScalarParameter() {defaultValue=500, group=SuperLevelling_Mask, parameterName=SL_StopAtLevel, sortPriority=3}
  n12 = Divide(A=Subtract(A=Clamp(None=@n9, Min=@n10, Max=@n11), B=@n10), B=Subtract(A=@n11, B=@n10))
  n13 = Max(A=Ceil(None=ScalarParameter() {defaultValue=1, group=SuperLevelling_Mask, parameterName=SL_QuantizationSteps, sliderMax=100, sliderMin=1}))
  n14 = Floor(None=Multiply(A=@n12, B=@n13))
  n15 = Ceil(None=Subtract(A=@n13))
  n16 = CurveAtlasRowParameter(CurveTime=LinearInterpolate(A=@n12, B=Divide(A=@n14, B=@n13), Alpha=Saturate(None=@n15))) {
   atlas=/BRCosmetics/Characters/Player/Female/Medium/Bodies/F_MED_BrutalBurglar/BrutalBurglarCurveAtlas
   curve=/BRCosmetics/Athena/LinearColorCurves/Curve_SuperlevelMask_BrutalKill
   group=SuperLevelling_Mask
   parameterName=SL_SuperLevelMaskCurve
  }
  n17 = ChannelMaskParameter(None=@n16) {group=SuperLevelling_Mask, parameterName=SL_SuperLevelCurveChannel}
  n18 = Power(Base=@n17.R, Exp=ScalarParameter() {defaultValue=2, group=SuperLevelling_Mask, parameterName=SL_MaskPos_Exponent})
  n19 = MaterialFunctionCall(
   A=VectorParameter() {group=SuperLevelling_Mask, parameterName=SL_MaskPos_01, defaultValue=<rgba -128 0 64 1>}.RGB
   B=VectorParameter() {group=SuperLevelling_Mask, parameterName=SL_MaskPos_02, defaultValue=<rgba -128 0 64 1>}.RGB
   C=VectorParameter() {group=SuperLevelling_Mask, parameterName=SL_MaskPos_03, defaultValue=<rgba -128 0 64 1>}.RGB
   Alpha=@n18
  ) {materialFunction=&Lerp_3Color}
  n20 = LinearInterpolate(
   A=ScalarParameter() {group=SuperLevelling_Mask, parameterName=SL_StartRadius}
   B=ScalarParameter() {defaultValue=512, group=SuperLevelling_Mask, parameterName=SL_EndRadius}
   Alpha=@n17.R
  )
  n21 = ScalarParameter() {defaultValue=0.5, group=SuperLevelling_Mask, parameterName=SL_CoverMaskHardness}
  n22 = TextureObjectParameter() {group=SuperLevelling_Mask, parameterName=SL_TransitionNoise, texture=/Game/Global/Textures/Default/Noise/T_SoftNoise}
  n23 = VectorParameter() {group=SuperLevelling_Mask, parameterName=SL_TransitionNoise_Tiling, defaultValue=<rgba 24 24 24 1>}
  n24 = FunctionInput(Preview=VertexInterpolator(VS=PreSkinnedNormal()).PS) {bUsePreviewValueAsDefault=true, inputName=Local Normal}
  n25 = Multiply(
   A=ChannelMaskParameter(None=MaterialFunctionCall(TextureObject=@n22, TextureSize=@n23.RGB, WorldPosition=@n3, World Space Normal=@n24) {materialFunction=/Game/Weapons/WeaponSkins/Wraps/MaterialFunctions/WorldAlignedTexture}.XYZ Texture) {group=SuperLevelling_Mask, parameterName=SL_TransitionNoiseChannel}.R
   B=LinearInterpolate(
    A=ScalarParameter() {defaultValue=0.5, group=SuperLevelling_Mask, parameterName=SL_Transition_Falloff_Start}
    B=ScalarParameter() {defaultValue=0.1, group=SuperLevelling_Mask, parameterName=SL_Transition_Falloff_End}
    Alpha=@n18
   )
  )
  n26 = Subtract(A=SphereMask(A=@n3, B=@n19.Result, Radius=@n20, Hardness=@n21), B=@n25)
  n27 = Saturate(None=StaticSwitchParameter(True=Round(None=@n26), False=@n26) {defaultValue=true, group=SuperLevelling_Mask, parameterName=SL_UseHardEdgedMask})
  n28 = Distance(A=@n26, B=ScalarParameter() {defaultValue=0.5, group=SuperLevelling_Mask, parameterName=SL_EdgeSamplePosition})
  n29 = ScalarParameter() {defaultValue=0.05, group=SuperLevelling_Mask, parameterName=SL_EdgeWidth}
  n30 = Clamp(None=OneMinus(None=Divide(A=@n28, B=@n29)))
  FunctionOutput(None=MaterialFunctionCall(X=@n27, Y=@n30, Z=Constant()) {materialFunction=&MakeFloat3}.Result) {outputName=SuperLevelling_Mask}
  n31 = GetMaterialAttributes(None=FunctionInput(Preview=MakeMaterialAttributes!(
     BaseColor
     Metallic
     Specular
     Roughness
     Anisotropy
     EmissiveColor
     Opacity
     OpacityMask
     Normal
     Tangent
     WorldPositionOffset
     SubsurfaceColor
     ClearCoat
     ClearCoatRoughness
     AmbientOcclusion
     Refraction
     CustomizedUVs_0
     CustomizedUVs_1
     CustomizedUVs_2
     CustomizedUVs_3
     CustomizedUVs_4
     CustomizedUVs_5
     CustomizedUVs_6
     CustomizedUVs_7
     PixelDepthOffset
     ShadingModel
     Displacement
    )) {bUsePreviewValueAsDefault=true, inputName=Material, inputType=FunctionInput_MaterialAttributes}) {attributeGetTypes [#1]}
  n32 = Multiply(
   A=Multiply(A=VectorParameter() {group=SuperLevelling_Mask, parameterName=SL_EdgeGlowColor, defaultValue=<rgba 1 0.4812 0 1>}.RGB, B=@n30)
   B=ScalarParameter() {defaultValue=5, group=SuperLevelling_Mask, parameterName=SL_EdgeGlowStrength}
  )
  n33 = ScalarParameter() {defaultValue=2, group=SuperLevelling_Mask, parameterName=SL_EdgeGlowPower}
  n34 = TextureSampleParameter2D() {
   group=SuperLevelling_Mask
   parameterName=SL_EffectMaskRegion
   samplerType=SAMPLERTYPE_Masks
   sortPriority=300
   texture=&T_WhiteMask
  }
  n35 = ChannelMaskParameter(None=@n34.RGB) {group=SuperLevelling_Mask, parameterName=SL_EffectMaskRegion_Channel, sortPriority=301}
  n36 = LinearInterpolate(A=@n35.R, B=OneMinus(None=@n35.R), Alpha=ScalarParameter() {group=SuperLevelling_Mask, parameterName=SL_InvertEffectMask, sortPriority=302})
  n37 = OneMinus(None=ChannelMaskParameter(None=@n34.RGB) {group=SuperLevelling_Mask, parameterName=SL_EffectMaskRegion_Channel_SubtractionChannel, sortPriority=301}.R)
  n38 = Power(Base=@n8, Exp=ScalarParameter() {defaultValue=1.33, group=SuperLevelling_Mask, parameterName=SL_Animation_EaseIn, sortPriority=103}) {constExponent=1.33}
  n39 = Abs(None=Multiply(A=Subtract(A=@n38) {constB=0.5}) {constB=2})
  n40 = MaterialFunctionCall(Float3=PreSkinnedPosition()) {materialFunction=&BreakOutFloat3Components}
  n41 = Multiply(
   A=Multiply(A=Add(A=@n40.R, B=@n40.R), B=@n40.B)
   B=ScalarParameter() {defaultValue=0.1, group=SuperLevelling_Mask, parameterName=AlwaysOnEmissiveGlow_OffsetScale}
  )
  n42 = VertexInterpolator(VS=Multiply(A=Add(A=@n5, B=@n41), B=ScalarParameter() {defaultValue=0.1, group=SuperLevelling_Mask, parameterName=AlwaysOnEmissiveGlow_PulseRate}))
  n43 = Clamp(None=Abs(None=Sine(None=@n42.PS)), Min=ScalarParameter() {group=SuperLevelling_Mask, parameterName=AlwaysOnEmissiveGlow_Min})
  n44 = Constant() {r=1}
  n45 = Power(
   Base=LinearInterpolate(
    A=OneMinus(None=@n39)
    B=StaticSwitchParameter(True=@n43, False=@n44) {group=SuperLevelling_Mask, parameterName=Vary_AlwaysOnGlow}
    Alpha=ScalarParameter() {group=SuperLevelling_Mask, parameterName=SL_AlwaysShowEmissiveEdge}
   )
   Exp=ScalarParameter() {defaultValue=6, group=SuperLevelling_Mask, parameterName=SL_Animation_EaseInOut, sortPriority=103}
  ) {constExponent=8}
  n46 = Multiply(A=StaticSwitchParameter(True=Subtract(A=@n36, B=@n37), False=@n36) {group=SuperLevelling_Mask, parameterName=SL_TextileHandMasking}, B=@n45)
  n47 = ScalarParameter() {group=SuperLevelling_Mask, parameterName=SL_ShiftedEdge_Shift}
  n48 = Subtract(A=SphereMask(A=@n3, B=@n19.Result, Radius=Add(A=@n47, B=@n20), Hardness=@n21), B=@n25)
  n49 = StaticSwitchParameter(True=Round(None=@n48), False=@n48) {defaultValue=true, group=SuperLevelling_Mask, parameterName=SL_Shifted_Edge_UseHardEdgedMask}
  n50 = LinearInterpolate(
   A=@n31.EmissiveColor
   B=Multiply(
    A=Power(Base=@n32, Exp=@n33)
    B=StaticSwitchParameter(True=@n46, False=@n45) {group=SuperLevelling_Mask, parameterName=SL_UseMaskForTransition}
   )
   Alpha=StaticSwitchParameter(True=Saturate(None=@n49), False=@n27) {group=SuperLevelling_Mask, parameterName=SL_Shift LeadingEdge}
  )
  FunctionOutput(None=SetMaterialAttributes!(
    MaterialAttributes=@n31.MaterialAttributes
    EmissiveColor=StaticSwitchParameter(True=@n50, False=@n31.EmissiveColor) {group=SuperLevelling_Mask, parameterName=SL_EnableEdgeGlow}
   ) {attributeSetTypes [#1]})
 ]}
