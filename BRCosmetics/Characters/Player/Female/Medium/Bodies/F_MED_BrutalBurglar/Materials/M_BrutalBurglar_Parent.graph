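# Parent material for the Brutal Burglar medium female character skin, adding optional hair shading and a gold super-levelling effect to the shared 2023 character parent.
# Output from the character PreFX function passes through static switches for Brutal Burglar hair, optionally blending in FN Hair 2.0 by its hair mask and adding flyaway hair.
# Super-levelling mask and gold material functions then feed PostFX, using material attributes and an expression-driven shading model.
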Material /BRCosmetics/Characters/Player/Female/Medium/Bodies/F_MED_BrutalBurglar/Materials/M_BrutalBurglar_Parent
graph {
 properties {
  bUseMaterialAttributes=true
  bUsedWithMeshDeformer=true
  bUsedWithMorphTargets=true
  bUsedWithSkeletalMesh=true
  pixelDepthOffsetMode=PDOM_Legacy
  shadingModel=MSM_FromMaterialExpression
  usedShadingModels=Default Lit | Subsurface | Clear Coat | Subsurface Profile | Cloth
 }
 nodes [
  n1 = NamedRerouteUsage()
  n2 = MaterialFunctionCall!(In=@n1, Mask Input=NamedRerouteUsage()) {materialFunction=/Game/Global/Materials/Character_Parent_2023/Hair_Parent_2023/MaterialFunctions/MF_FN_Hair_2023}
  n3 = BlendMaterialAttributes(
   A=@n1
   B=StaticSwitchParameter(True=@n2.Result, False=@n1) {defaultValue=true, group=Skin FX, parameterName=UseFNHair2_0}
   Alpha=@n2.HairMask
  )
  n4 = MaterialFunctionCall!(MaterialInput=@n3) {materialFunction=/Game/Global/Materials/Character_Parent_2023/Hair_Parent_2023/MaterialFunctions/MF_FlyAwayHair_2023}
  n5 = MaterialFunctionCall!(
   Local Normal
   Local Position=NamedRerouteUsage()
   Material=StaticSwitchParameter(True=StaticSwitchParameter(True=@n4.Result, False=@n3) {group=Skin FX, parameterName=IsFlyAwayHair?}, False=@n1) {parameterName=IsBrutalBurglarHair}
  ) {materialFunction=&/MF_SuperLevellingMask_BrutalBurglar}
  n6 = MaterialFunctionCall() {materialFunction=&MF_CharacterParent_2023_PreFX}
  MaterialFunctionCall(
   Material=MaterialFunctionCall!(
    SuperLevellingMask=@n5.SuperLevelling_Mask
    SkinFXMask=AppendVector(A=@n6.Tech Art Mask, B=@n6.Tech Art Mask)
    GMap=NamedRerouteUsage()
    Material=@n5.SuperLevelling_Mask
   ) {materialFunction=&/MF_BrutalBurglar_GoldMaterial}.Result
   M=NamedRerouteUsage()
   Pre Skinned Local Position=@n6.Pre skinned Local Position
   Distance Blend=@n6.Pre skinned Local Position
   Fresnel=@n6.Pre skinned Local Position
  ) {materialFunction=&MF_CharacterParent_2023_PostFX}
  NamedRerouteDeclaration(None=@n6.Result) {name=OriginalMA, nodeColor=<rgba 1 0.6328 0 1>}
  NamedRerouteDeclaration(None=@n6.M) {name=InMask, nodeColor=<rgba 0.3125 0 1 1>}
  NamedRerouteDeclaration(None=@n6.Pre skinned Local Position) {name=PreskinnedLocalPosition, nodeColor=<rgba 0.0078 1 0 1>}
 ]
}
