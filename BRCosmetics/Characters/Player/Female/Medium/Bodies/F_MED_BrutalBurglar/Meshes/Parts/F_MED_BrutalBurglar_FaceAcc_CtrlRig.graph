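# Control Rig for the BrutalBurglar female face accessory that applies pose-driven corrective rotations to hair bones.
# A spherical pose reader measures the head against spine_05 and weights a chain of Modify Transforms nodes that rotate the right front strand, right side and back hair.
# A second reader for the right upper arm relative to the head drives larger side and back rotations, but its chain is not connected to execution.
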ControlRigBlueprint /BRCosmetics/Characters/Player/Female/Medium/Bodies/F_MED_BrutalBurglar/Meshes/Parts/F_MED_BrutalBurglar_FaceAcc_CtrlRig
graphs [
 RigVMFunctionLibrary {links [], nodes []}
 RigVMModel {nodes [
   RigUnit_BeginExecution = &RigUnit_BeginExecution()
   RigVMFunction_Sequence = &RigVMFunction_Sequence(ExecuteContext=@RigUnit_BeginExecution.ExecutePin)
   n1 = &RigUnit_SphericalPoseReader(
    ActiveRegionScaleFactors=(PositiveWidth=2,NegativeWidth=2,PositiveHeight=3,NegativeHeight=3)
    ActiveRegionSize=0.1
    Debug=(bDrawDebug=false,bDraw2D=false,bDrawLocalAxes=false,DebugScale=25,DebugSegments=12,DebugThickness=0.15)
    DriverAxis=(X=0,Y=1,Z=0)
    DriverItem=(Type=Bone,Name="head")
    FalloffRegionScaleFactors=(PositiveWidth=1,NegativeWidth=1,PositiveHeight=2,NegativeHeight=2)
    FalloffSize=0.4
    FlipWidthScaling=false
    OptionalParentItem=(Type=Bone,Name="spine_05")
    RotationOffset=(X=83.1834,Y=-71.2041,Z=-8.5636)
    ExecutePin=@RigVMFunction_Sequence.A
   )
   > &RigUnit_ModifyTransforms(
    ItemToModify=((Item=(Type=Bone,Name="dyn_fr_strand_r_01"),Transform=(Rotation=(X=0.0076,Y=0.0868,Z=-0.0868,W=0.9924),Translation=(X=0,Y=0,Z=0),Scale3D=(X=1,Y=1,Z=1))))
    WeightMinimum=0
    Weight=@n1.OutputParam
   )
   > &RigUnit_ModifyTransforms(
    ItemToModify=((Item=(Type=Bone,Name="dyn_side_r1_MSTR"),Transform=(Rotation=(X=0.0076,Y=0.0868,Z=-0.0868,W=0.9924),Translation=(X=0,Y=0,Z=0),Scale3D=(X=1,Y=1,Z=1))))
    WeightMinimum=0
    Weight=@n1.OutputParam
   )
   > &RigUnit_ModifyTransforms(
    ItemToModify=((Item=(Type=Bone,Name="dyn_side_r2_MSTR"),Transform=(Rotation=(X=0,Y=0.0872,Z=0,W=0.9962),Translation=(X=0,Y=0,Z=0),Scale3D=(X=1,Y=1,Z=1))))
    WeightMinimum=0
    Weight=@n1.OutputParam
   )
   > &RigUnit_ModifyTransforms(
    ItemToModify=((Item=(Type=Bone,Name="dyn_back_MSTR"),Transform=(Rotation=(X=0,Y=0.0872,Z=0,W=0.9962),Translation=(X=0,Y=0,Z=0),Scale3D=(X=1,Y=1,Z=1))))
    WeightMinimum=0
    Weight=@n1.OutputParam
   )
   n2 = &RigUnit_SphericalPoseReader(
    ActiveRegionScaleFactors=(PositiveWidth=1,NegativeWidth=1,PositiveHeight=1,NegativeHeight=1)
    ActiveRegionSize=0.1
    Debug=(bDrawDebug=True,bDraw2D=False,bDrawLocalAxes=False,DebugScale=25,DebugSegments=20,DebugThickness=0.25)
    DriverAxis=(X=0,Y=1,Z=0)
    DriverItem=(Type=Bone,Name="upperarm_r")
    ExecutePin=()
    FalloffRegionScaleFactors=(PositiveWidth=1,NegativeWidth=1,PositiveHeight=1,NegativeHeight=1)
    FalloffSize=0.4
    FlipHeightScaling=False
    FlipWidthScaling=false
    OptionalParentItem=(Type=Bone,Name="head")
    RotationOffset=(X=20.3647,Y=-123.5472,Z=84.3604)
   )
   > &RigUnit_ModifyTransforms(
    ItemToModify=((Item=(Type=Bone,Name="dyn_side_r1_MSTR"),Transform=(Rotation=(X=0,Y=0.2588,Z=0,W=0.9659),Translation=(X=0,Y=0,Z=0),Scale3D=(X=1,Y=1,Z=1))))
    WeightMinimum=0
    Weight=@n2.OutputParam
   )
   > &RigUnit_ModifyTransforms(
    ItemToModify=((Item=(Type=Bone,Name="dyn_side_r1_MSTR"),Transform=(Rotation=(X=0,Y=0.2588,Z=0,W=0.9659),Translation=(X=0,Y=0,Z=0),Scale3D=(X=1,Y=1,Z=1))))
    WeightMinimum=0
    Weight=@n2.OutputParam
   )
   > &RigUnit_ModifyTransforms(
    ItemToModify=((Item=(Type=Bone,Name="dyn_back_01"),Transform=(Rotation=(X=0,Y=0.2588,Z=0,W=0.9659),Translation=(X=0,Y=0,Z=0),Scale3D=(X=1,Y=1,Z=1))))
    WeightMinimum=0
    Weight=@n2.OutputParam
   )
  ]}
]
hierarchy [
 root {loc=<0 0 0>, rot=<0 0 0 1>}
 pelvis {parent=root, loc=<0 -0.5301 98.0571>, rot=<0.0127 0.707 -0.0127 -0.707>}
 spine_01 {parent=pelvis, loc=<6.1208 -0.9328 0>, rot=<0 0 0.034 -0.9994>}
 spine_02 {parent=spine_01, loc=<6.5997 -0.2095 0>, rot=<0 0 0.0238 -0.9997>}
 spine_03 {parent=spine_02, loc=<6.4914 -0.1237 0>, rot=<0 0 0.0149 -0.9999>}
 spine_04 {parent=spine_03, loc=<6.9598 0.0794 0>, rot=<0 0 -0.0132 -0.9999>}
 spine_05 {parent=spine_04, loc=<6.9987 0 0>, rot=<0 0 -0.0015 -1>}
 neck_01 {parent=spine_05, loc=<11.3138 0.1529 0>, rot=<0 0 -0.2003 -0.9797>}
 neck_02 {parent=neck_01, loc=<4.5627 0 0>, rot=<0 0 0.0504 -0.9987>}
 head {parent=neck_02, loc=<5.3914 0 0>, rot=<0 0 0.115 -0.9934>}
 hair_MSTR {parent=head, loc=<0 0 0>, rot=<0 0 0 1>}
 dyn_fr_strand_l_MSTR {parent=hair_MSTR, loc=<9.6035 8.5276 -1.3005>, rot=<-0.2683 -0.0248 0.9478 -0.1707>}
 dyn_fr_strand_l_01 {parent=dyn_fr_strand_l_MSTR, loc=<0 0 0>, rot=<0 0 0 1>}
 dyn_fr_strand_l_02 {parent=dyn_fr_strand_l_01, loc=<6.1729 0 0>, rot=<0 -0.0189 0 -0.9998>}
 dyn_fr_strand_l_03 {parent=dyn_fr_strand_l_02, loc=<7.6513 0 0>, rot=<0 0 0 1>}
 dyn_back_MSTR {parent=hair_MSTR, loc=<13.4554 -4.0809 0.8907>, rot=<-0.0922 0.6899 0.7101 -0.1061>}
 dyn_back_01 {parent=dyn_back_MSTR, loc=<0 0 0>, rot=<0 0 0 1>}
 dyn_back_02 {parent=dyn_back_01, loc=<9.3247 0 -0.0889>, rot=<0 0 0 1>}
 dyn_back_03 {parent=dyn_back_02, loc=<8.4387 0 0.1462>, rot=<0 0 0 1>}
 dyn_back_04 {parent=dyn_back_03, loc=<7.9461 -0.0607 0.1128>, rot=<0 0 0 1>}
 dyn_side_r2_MSTR {parent=hair_MSTR, loc=<11.385 -5.227 4.5176>, rot=<0.1122 -0.8864 -0.4377 0.1006>}
 dyn_side_r2_01 {parent=dyn_side_r2_MSTR, loc=<0 0 0>, rot=<0 0 0 1>}
 dyn_side_r2_02 {parent=dyn_side_r2_01, loc=<17.1879 0 -0.4161>, rot=<0 -0.0189 0 -0.9998>}
 dyn_side_l1_MSTR {parent=hair_MSTR, loc=<11.7099 -4.4698 -4.2106>, rot=<-0.0889 0.2287 0.9687 -0.0374>}
 dyn_side_l1_01 {parent=dyn_side_l1_MSTR, loc=<0 0 0>, rot=<0 0 0 1>}
 dyn_side_l1_02 {parent=dyn_side_l1_01, loc=<9.967 0.0218 -0.1225>, rot=<0 -0.0189 0 -0.9998>}
 dyn_side_l1_03 {parent=dyn_side_l1_02, loc=<9.2891 -0.0218 -0.0208>, rot=<0 0 0 1>}
 dyn_side_l1_04 {parent=dyn_side_l1_03, loc=<6.6635 0 0>, rot=<0 0 0 1>}
 dyn_side_r1_MSTR {parent=hair_MSTR, loc=<11.8113 -1.0141 6.1944>, rot=<0.0713 -0.9826 -0.1323 0.109>}
 dyn_side_r1_01 {parent=dyn_side_r1_MSTR, loc=<0 0 0>, rot=<0 0 0 1>}
 dyn_side_r1_02 {parent=dyn_side_r1_01, loc=<17.2033 0 -0.4167>, rot=<0 -0.0189 0 -0.9998>}
 dyn_fr_strand_r_MSTR {parent=hair_MSTR, loc=<9.924 1.8949 5.3568>, rot=<0.1169 -0.98 0.0379 0.1563>}
 dyn_fr_strand_r_01 {parent=dyn_fr_strand_r_MSTR, loc=<0 0 0>, rot=<0 0 0 1>}
 dyn_fr_strand_r_02 {parent=dyn_fr_strand_r_01, loc=<12.9341 0 0>, rot=<0 0 0 1>}
 dyn_side_l2_MSTR {parent=hair_MSTR, loc=<9.0973 1.3291 -6.5723>, rot=<-0.0692 -0.0041 0.9953 -0.0681>}
 dyn_side_l2_01 {parent=dyn_side_l2_MSTR, loc=<0 0 0>, rot=<0 0 0 1>}
 dyn_side_l2_02 {parent=dyn_side_l2_01, loc=<8.3898 0 -0.0837>, rot=<0 -0.0189 0 -0.9998>}
 dyn_side_l2_03 {parent=dyn_side_l2_02, loc=<7.6336 0 0>, rot=<0 0 0 1>}
 dyn_side_l2_04 {parent=dyn_side_l2_03, loc=<7.1501 0 0>, rot=<0 0 0 1>}
 clavicle_l {parent=spine_05, loc=<5.4831 -0.1053 -3.2946>, rot=<0.0465 -0.7289 -0.009 -0.6829>}
 upperarm_l {parent=clavicle_l, loc=<9.8658 -1.2136 -0.4105>, rot=<0.0164 -0.3245 -0.0269 -0.9454>}
 clavicle_r {parent=spine_05, loc=<5.4831 -0.1053 3.2946>, rot=<0.7289 0.0465 -0.6829 0.009>}
 upperarm_r {parent=clavicle_r, loc=<-9.8658 1.2136 0.4105>, rot=<0.0164 -0.3245 -0.0269 -0.9454>}
]
