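# Physics asset for the BrutalBurglar female character's FaceAcc mesh part, setting up hair physics with capsule bodies.
# Head, neck, upper spine, clavicle and upper arm capsules form an angularly free body chain, with hair_MSTR attached to the head.
# Back, side and front strand chains use limited swing and twist, though the right front and first right side masters lack constraints to hair_MSTR.
# Angularly free cross-links join both right side chains and the back chain.
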PhysicsAsset /BRCosmetics/Characters/Player/Female/Medium/Bodies/F_MED_BrutalBurglar/Meshes/Parts/F_MED_BrutalBurglar_FaceAcc_Physics
bodies {
 clavicle_l [{shapeType=Capsule, length=12.9911, radius=4.9683, shapeName=clavicle_l_capsule, center=<3.9317 0 0>, rotation=<rot 90 0 0>}]
 clavicle_r [{shapeType=Capsule, length=9.5989, radius=6.6866, shapeName=clavicle_r_capsule, center=<-3.4847 0 -0.102>, rotation=<rot 90 0 0>}]
 dyn_back_01 [{shapeType=Capsule, length=3.72, radius=2.0857, shapeName=dyn_back_01_capsule, center=<4.0452 0 0>, rotation=<rot 0 0 90>}]
 dyn_back_02 [{shapeType=Capsule, length=6.2916, radius=2.8835, shapeName=dyn_back_02_capsule, center=<4.7813 0 0>, rotation=<rot 0 0 90>}]
 dyn_back_03 [{shapeType=Capsule, length=8.9106, radius=3.3904, shapeName=dyn_back_03_capsule, center=<3.8659 0 0>, rotation=<rot 0 0 90>}]
 dyn_back_04 [{
   shapeType=Capsule
   length=9.0852
   radius=2.6948
   shapeName=dyn_back_04_capsule
   center=<5.7232 -0.0527 0.0003>
   rotation=<rot -0.0045 179.0955 90>
  }]
 dyn_back_MSTR [{shapeType=Capsule, length=1, radius=1, shapeName=dyn_back_MSTR_capsule, rotation=<rot 0 0 90>}]
 dyn_fr_strand_l_01 [{shapeType=Capsule, length=5.8225, radius=1.1775, shapeName=dyn_fr_strand_l_01_capsule, rotation=<rot 0 0 90>}]
 dyn_fr_strand_l_02 [{shapeType=Capsule, length=5, radius=2, shapeName=dyn_fr_strand_l_02_capsule, rotation=<rot 0 0 90>}]
 dyn_fr_strand_l_03 [{shapeType=Capsule, length=5, radius=2, shapeName=dyn_fr_strand_l_03_capsule, rotation=<rot 0 0 90>}]
 dyn_fr_strand_l_MSTR [{shapeType=Capsule, length=1, radius=1, shapeName=dyn_fr_strand_l_MSTR_capsule, rotation=<rot 0 0 90>}]
 dyn_fr_strand_r_01 [{shapeType=Capsule, length=3.5864, radius=3.4136, shapeName=dyn_fr_strand_r_01_capsule, center=<8.9901 0 0>, rotation=<rot 0 0 90>}]
 dyn_fr_strand_r_02 [{shapeType=Capsule, length=3.2298, radius=3.7702, shapeName=dyn_fr_strand_r_02_capsule, center=<3.8258 0 0>, rotation=<rot 0 0 90>}]
 dyn_fr_strand_r_MSTR [{shapeType=Capsule, length=0.505, radius=0.505, shapeName=dyn_fr_strand_r_MSTR_capsule, rotation=<rot 0 -90 0>}]
 dyn_side_l1_01 [{shapeType=Capsule, length=6.3127, radius=0.6873, shapeName=dyn_side_l1_01_capsule, rotation=<rot 0 0 90>}]
 dyn_side_l1_02 [{shapeType=Capsule, length=4.4124, radius=2.5876, shapeName=dyn_side_l1_02_capsule, center=<1.9201 0 0>, rotation=<rot 0 0 90>}]
 dyn_side_l1_03 [{shapeType=Capsule, length=3.6723, radius=3.3277, shapeName=dyn_side_l1_03_capsule, center=<3.7142 0 0>, rotation=<rot 0 0 90>}]
 dyn_side_l1_MSTR [{shapeType=Capsule, length=1, radius=1, shapeName=dyn_side_l1_MSTR_capsule, rotation=<rot 0 0 90>}]
 dyn_side_l2_01 [{shapeType=Capsule, length=5.8225, radius=1.1775, shapeName=dyn_side_l2_01_capsule, rotation=<rot 0 0 90>}]
 dyn_side_l2_02 [{shapeType=Capsule, length=5, radius=2, shapeName=dyn_side_l2_02_capsule, center=<1.9201 0 0>, rotation=<rot 0 0 90>}]
 dyn_side_l2_03 [{shapeType=Capsule, length=3.6723, radius=3.3277, shapeName=dyn_side_l2_03_capsule, center=<3.7142 0 0>, rotation=<rot 0 0 90>}]
 dyn_side_l2_MSTR [{shapeType=Capsule, length=1, radius=1, shapeName=dyn_side_l2_MSTR_capsule, rotation=<rot 0 0 90>}]
 dyn_side_r1_01 [{
   shapeType=Capsule
   length=6.1954
   radius=2.9595
   shapeName=dyn_side_r1_01_capsule
   center=<9.6754 -0.0455 0>
   rotation=<rot -0.0001 89.2824 90.0001>
  }]
 dyn_side_r1_02 [{shapeType=Capsule, length=2.1277, radius=4.8723, shapeName=dyn_side_r1_02_capsule, center=<0.4433 0 0>, rotation=<rot 0 0 90>}]
 dyn_side_r1_MSTR [{shapeType=Capsule, length=0.505, radius=0.505, shapeName=dyn_side_r1_MSTR_capsule, rotation=<rot 0 -90 0>}]
 dyn_side_r2_01 [{shapeType=Capsule, length=7.5557, radius=3.2126, shapeName=dyn_side_r2_01_capsule, center=<7.9683 0 0>, rotation=<rot 0 94.1991 90.0002>}]
 dyn_side_r2_02 [{shapeType=Capsule, length=3.2407, radius=3.7593, shapeName=dyn_side_r2_02_capsule, center=<3.7331 0 0>, rotation=<rot 0 0 90>}]
 dyn_side_r2_MSTR [{shapeType=Capsule, length=0.505, radius=0.505, shapeName=dyn_side_r2_MSTR_capsule, rotation=<rot 0 -90 0>}]
 hair_MSTR [{shapeType=Capsule, length=0.505, radius=0.505, shapeName=hair_MSTR_capsule, rotation=<rot 0 -90 0>}]
 head [{shapeType=Capsule, length=6.9444, radius=6.0181, shapeName=head_capsule, center=<1.4505 0.7996 0>, rotation=<rot 90 0 0>}]
 neck_01 [{shapeType=Capsule, length=2.673, radius=5.808, shapeName=neck_01_capsule, rotation=<rot 90 0 0>}]
 neck_02 [{shapeType=Capsule, length=3.7621, radius=5.2694, shapeName=neck_02_capsule, rotation=<rot 90 0 0>}]
 spine_05 [{shapeType=Capsule, length=19.1141, radius=8.9283, shapeName=spine_05_capsule, rotation=<rot 0.0795 -0.0002 0>}]
 upperarm_l [{shapeType=Capsule, length=12.608, radius=4.9687, shapeName=upperarm_l_capsule, center=<6.1211 0 0>, rotation=<rot 90 0 0>}]
 upperarm_r [{shapeType=Capsule, length=16.6915, radius=5.0534, shapeName=upperarm_r_capsule, center=<-5.9701 0 0>, rotation=<rot 90 0 0>}]
}
constraints [
 {bone1Name=dyn_fr_strand_l_MSTR, bone2Name=hair_MSTR, swing1LimitDegrees=25, swing2LimitDegrees=35, twistLimitDegrees=2}
 {bone1Name=dyn_fr_strand_l_01, bone2Name=dyn_fr_strand_l_MSTR, swing1LimitDegrees=10, swing2LimitDegrees=15, twistLimitDegrees=2}
 {bone1Name=dyn_fr_strand_l_02, bone2Name=dyn_fr_strand_l_01, swing1LimitDegrees=10, swing2LimitDegrees=15, twistLimitDegrees=2}
 {bone1Name=dyn_fr_strand_l_03, bone2Name=dyn_fr_strand_l_02, swing1LimitDegrees=25, swing2LimitDegrees=35, twistLimitDegrees=2}
 {bone1Name=dyn_side_l1_MSTR, bone2Name=hair_MSTR, swing1LimitDegrees=60, swing2LimitDegrees=45, twistLimitDegrees=2}
 {bone1Name=dyn_side_l1_01, bone2Name=dyn_side_l1_MSTR, swing1LimitDegrees=60, swing2LimitDegrees=45, twistLimitDegrees=2}
 {bone1Name=dyn_side_l2_MSTR, bone2Name=hair_MSTR, swing1LimitDegrees=25, swing2LimitDegrees=35, twistLimitDegrees=2}
 {bone1Name=dyn_side_l2_01, bone2Name=dyn_side_l2_MSTR, swing1LimitDegrees=45, swing2LimitDegrees=15, twistLimitDegrees=2}
 {
  bone1Name=hair_MSTR
  bone2Name=head
  swing1LimitDegrees=45
  swing1Motion=Free
  swing2LimitDegrees=45
  swing2Motion=Free
  twistLimitDegrees=45
  twistMotion=Free
 }
 {
  bone1Name=clavicle_l
  bone2Name=spine_05
  swing1LimitDegrees=45
  swing1Motion=Free
  swing2LimitDegrees=45
  swing2Motion=Free
  twistLimitDegrees=45
  twistMotion=Free
 }
 {
  bone1Name=upperarm_l
  bone2Name=clavicle_l
  swing1LimitDegrees=45
  swing1Motion=Free
  swing2LimitDegrees=45
  swing2Motion=Free
  twistLimitDegrees=45
  twistMotion=Free
 }
 {
  bone1Name=clavicle_r
  bone2Name=spine_05
  swing1LimitDegrees=45
  swing1Motion=Free
  swing2LimitDegrees=45
  swing2Motion=Free
  twistLimitDegrees=45
  twistMotion=Free
 }
 {
  bone1Name=upperarm_r
  bone2Name=clavicle_r
  swing1LimitDegrees=45
  swing1Motion=Free
  swing2LimitDegrees=45
  swing2Motion=Free
  twistLimitDegrees=45
  twistMotion=Free
 }
 {
  bone1Name=neck_01
  bone2Name=spine_05
  swing1LimitDegrees=45
  swing1Motion=Free
  swing2LimitDegrees=45
  swing2Motion=Free
  twistLimitDegrees=45
  twistMotion=Free
 }
 {
  bone1Name=neck_02
  bone2Name=neck_01
  swing1LimitDegrees=45
  swing1Motion=Free
  swing2LimitDegrees=45
  swing2Motion=Free
  twistLimitDegrees=45
  twistMotion=Free
 }
 {
  bone1Name=head
  bone2Name=neck_02
  swing1LimitDegrees=45
  swing1Motion=Free
  swing2LimitDegrees=45
  swing2Motion=Free
  twistLimitDegrees=45
  twistMotion=Free
 }
 {bone1Name=dyn_back_MSTR, bone2Name=hair_MSTR, swing1LimitDegrees=60, swing2LimitDegrees=60, twistLimitDegrees=15}
 {bone1Name=dyn_back_01, bone2Name=dyn_back_MSTR, swing1LimitDegrees=60, swing2LimitDegrees=60, twistLimitDegrees=15}
 {bone1Name=dyn_back_02, bone2Name=dyn_back_01, swing1LimitDegrees=60, swing2LimitDegrees=60, twistLimitDegrees=15}
 {bone1Name=dyn_back_03, bone2Name=dyn_back_02, swing1LimitDegrees=40.68, swing2LimitDegrees=47.88, twistLimitDegrees=15}
 {bone1Name=dyn_back_04, bone2Name=dyn_back_03, swing1LimitDegrees=90, swing2LimitDegrees=90, twistLimitDegrees=15}
 {bone1Name=dyn_fr_strand_r_01, bone2Name=dyn_fr_strand_r_MSTR, swing1LimitDegrees=24.84, swing2LimitDegrees=45, twistLimitDegrees=45}
 {bone1Name=dyn_fr_strand_r_02, bone2Name=dyn_fr_strand_r_01, swing1LimitDegrees=14.76, swing2LimitDegrees=45, twistLimitDegrees=45}
 {bone1Name=dyn_side_r1_01, bone2Name=dyn_side_r1_MSTR, swing1LimitDegrees=45, swing2LimitDegrees=45, twistLimitDegrees=5}
 {bone1Name=dyn_side_r1_02, bone2Name=dyn_side_r1_01, swing1LimitDegrees=45, swing2LimitDegrees=45, twistLimitDegrees=5}
 {
  bone1Name=dyn_side_r2_MSTR
  bone2Name=hair_MSTR
  swing1LimitDegrees=45
  swing1Motion=Free
  swing2LimitDegrees=45
  swing2Motion=Free
  twistLimitDegrees=45
  twistMotion=Free
 }
 {bone1Name=dyn_side_r2_01, bone2Name=dyn_side_r2_MSTR, swing1LimitDegrees=45, swing2LimitDegrees=45, twistLimitDegrees=10}
 {bone1Name=dyn_side_r2_02, bone2Name=dyn_side_r2_01, swing1LimitDegrees=45, swing2LimitDegrees=45, twistLimitDegrees=10}
 {
  bone1Name=dyn_side_r1_02
  bone2Name=dyn_side_r2_02
  swing1LimitDegrees=45
  swing1Motion=Free
  swing2LimitDegrees=45
  swing2Motion=Free
  twistLimitDegrees=45
  twistMotion=Free
 }
 {
  bone1Name=dyn_side_r2_02
  bone2Name=dyn_back_03
  swing1LimitDegrees=45
  swing1Motion=Free
  swing2LimitDegrees=45
  swing2Motion=Free
  twistLimitDegrees=45
  twistMotion=Free
 }
 {bone1Name=dyn_side_l2_02, bone2Name=dyn_side_l2_01, swing1LimitDegrees=45, swing2LimitDegrees=45, twistLimitDegrees=1}
 {bone1Name=dyn_side_l2_03, bone2Name=dyn_side_l2_02, swing1LimitDegrees=45, swing2LimitDegrees=45, twistLimitDegrees=1}
 {bone1Name=dyn_side_l1_02, bone2Name=dyn_side_l1_01, swing1LimitDegrees=45, swing2LimitDegrees=45, twistLimitDegrees=1}
 {bone1Name=dyn_side_l1_03, bone2Name=dyn_side_l1_02, swing1LimitDegrees=45, swing2LimitDegrees=45, twistLimitDegrees=1}
]
