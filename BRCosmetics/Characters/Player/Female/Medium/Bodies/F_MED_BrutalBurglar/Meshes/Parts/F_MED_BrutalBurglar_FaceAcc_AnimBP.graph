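# Animation Blueprint for the Brutal Burglar face accessory part, adding secondary motion to its dynamic hair bones.
# It copies the parent mesh pose, then Backpack_On selects between a RigidBody branch using a cape physics asset and a Control Rig plus RigidBody branch.
# Constraints and additive bone offsets shape each branch and the blended result against spine and clavicles.
# The event graph updates backbling detection, state-based gravity override and an emote hair alpha.
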AnimBlueprint /BRCosmetics/Characters/Player/Female/Medium/Bodies/F_MED_BrutalBurglar/Meshes/Parts/F_MED_BrutalBurglar_FaceAcc_AnimBP
target_skeleton=&/F_MED_BrutalBurglar_FaceAcc_Skeleton
graphs [
 AnimGraph {nodes [
   Component To Local()
   n1 = "Constraint - dyn_side_r1_01 (PP)"!(ComponentPose, bAlphaBoolEnabled:bool=True, ConstraintWeights_0=1, ConstraintWeights_1:float=0) -> (Pose) {
    alphaInputType=Bool
    alphaBoolBlend {blendInTime=5, blendOption=QuarticInOut, blendOutTime=5}
    boneToModify {boneName=dyn_side_r1_01}
    constraintSetup [{targetBone {boneName=spine_05}}, {targetBone {boneName=clavicle_r}}]
    constraintWeights [1,1]
   }
   "Constraint - dyn_fr_strand_r_02 (PP)"!(ComponentPose=@n1, bAlphaBoolEnabled:bool=True, ConstraintWeights_0=1, ConstraintWeights_1:float=0) -> (Pose) {
    alphaInputType=Bool
    alphaBoolBlend {blendInTime=5, blendOption=QuarticInOut, blendOutTime=5}
    boneToModify {boneName=dyn_fr_strand_r_02}
    constraintSetup [{targetBone {boneName=spine_05}}, {targetBone {boneName=clavicle_r}}]
    constraintWeights [1,1]
   }
   "Constraint - dyn_side_r2_01 (PP)"!(ComponentPose, bAlphaBoolEnabled:bool=True, ConstraintWeights_0=1, ConstraintWeights_1:float=0) -> (Pose) {
    alphaInputType=Bool
    alphaBoolBlend {blendInTime=5, blendOption=QuarticInOut, blendOutTime=5}
    boneToModify {boneName=dyn_side_r2_01}
    constraintSetup [{targetBone {boneName=spine_05}}, {targetBone {boneName=clavicle_r}}]
    constraintWeights [1,1]
   }
   n2 = "Constraint - dyn_side_l1_03 (PP)"!(ComponentPose, bAlphaBoolEnabled:bool=True, ConstraintWeights_0=1, ConstraintWeights_1:float=0) -> (Pose) {
    alphaInputType=Bool
    alphaBoolBlend {blendOption=Linear}
    boneToModify {boneName=dyn_side_l1_03}
    constraintSetup [{targetBone {boneName=spine_05}}, {targetBone {boneName=head}}]
    constraintWeights [1,1]
   }
   n3 = "Constraint - dyn_side_l1_03 (PP)"(ComponentPose=@n2, bAlphaBoolEnabled=True, ConstraintWeights_0=1, ConstraintWeights_1=0) {
    alphaInputType=Bool
    alphaBoolBlend {blendOption=Linear}
    boneToModify {boneName=dyn_side_l1_03}
    constraintSetup [{targetBone {boneName=spine_05}}, {targetBone {boneName=clavicle_r}}]
    constraintWeights [1,1]
   }
   n4 = "Transform (Modify) Bone - Bone: dyn_side_r2_01"!(ComponentPose=@n3, bAlphaBoolEnabled:bool=True, Translation="0,0,0", Rotation="-30,-20,0", Scale="1,1,1") -> (Pose) {
    alphaInputType=Bool
    rotationMode=BMM_Additive
    rotationSpace=BCS_BoneSpace
    alphaBoolBlend {blendOption=Linear}
    boneToModify {boneName=dyn_side_r2_01}
    rotation {pitch=-30, yaw=-20}
   }
   n5 = "Transform (Modify) Bone - Bone: dyn_side_r1_01"!(ComponentPose=@n4, bAlphaBoolEnabled:bool=True, Translation="0,0,0", Rotation="-10,10,0", Scale="1,1,1") -> (Pose) {
    alphaInputType=Bool
    rotationMode=BMM_Additive
    rotationSpace=BCS_BoneSpace
    alphaBoolBlend {blendOption=Linear}
    boneToModify {boneName=dyn_side_r1_01}
    rotation {pitch=-10, yaw=10}
   }
   "Transform (Modify) Bone - Bone: dyn_fr_strand_r_01"!(ComponentPose=@n5, bAlphaBoolEnabled:bool=True, Translation="0,0,0", Rotation="0,10,0", Scale="1,1,1") -> (Pose) {
    alphaInputType=Bool
    rotationMode=BMM_Additive
    rotationSpace=BCS_BoneSpace
    alphaBoolBlend {blendOption=Linear}
    boneToModify {boneName=dyn_fr_strand_r_01}
    rotation {yaw=10}
   }
   n6 = RigidBody!(
    ComponentPose=Local To Component(LocalPose=Use cached pose 'base_pose'())
    Alpha
    OverrideWorldGravity:Vector=Get Gravity Override()
    ExternalForce
   ) -> (Pose) {
    bClampLinearTranslationLimitToRefPose=true
    bOverrideWorldGravity=true
    overridePhysicsAsset=&/F_MED_BrutalBurglar_FaceAcc_Physics_Cape
    simulationSpace=BaseBoneSpace
    alphaBoolBlend {blendOption=Linear}
    baseBoneRef {boneName=spine_05}
    overrideWorldGravity {z=-980}
    simSpaceSettings {
     dampingAlpha=0
     maxAngularAcceleration=100
     maxAngularVelocity=3
     maxLinearAcceleration=100
     maxLinearVelocity=700
     velocityScaleZ=2
     worldAlpha=0.5
     externalAngularVelocity=<2 2 2>
     externalLinearDragV=<2 2 2>
     externalLinearVelocity=<2 2 2>
    }
   }
   n7 = "Constraint - dyn_back_03 (P)"(ComponentPose=@n6, Alpha=1, ConstraintWeights_0=1) {
    alphaBoolBlend {blendOption=Linear}
    boneToModify {boneName=dyn_back_03}
    constraintSetup [{targetBone {boneName=spine_05}}]
    constraintWeights [1]
   }
   n8 = "Constraint - dyn_side_r2_02 (PP)"!(ComponentPose=@n7, Alpha=1, ConstraintWeights_0=0.5, ConstraintWeights_1:float=0.5) -> (Pose) {
    alphaBoolBlend {blendOption=Linear}
    boneToModify {boneName=dyn_side_r2_02}
    constraintSetup [{targetBone {boneName=spine_05}}, {targetBone {boneName=clavicle_r}}]
    constraintWeights [1,1]
   }
   n9 = "Constraint - dyn_side_r1_02 (PP)"!(ComponentPose=@n8, Alpha=1, ConstraintWeights_0=0.5, ConstraintWeights_1:float=0.5) -> (Pose) {
    alphaBoolBlend {blendOption=Linear}
    boneToModify {boneName=dyn_side_r1_02}
    constraintSetup [{targetBone {boneName=spine_05}}, {targetBone {boneName=clavicle_r}}]
    constraintWeights [1,1]
   }
   n10 = "Constraint - dyn_side_l2_03 (PP)"!(ComponentPose=@n9, Alpha=1, ConstraintWeights_0=0.5, ConstraintWeights_1:float=0.5) -> (Pose) {
    alphaBoolBlend {blendOption=Linear}
    boneToModify {boneName=dyn_side_l2_03}
    constraintSetup [{targetBone {boneName=spine_05}}, {targetBone {boneName=clavicle_l}}]
    constraintWeights [1,1]
   }
   n11 = "Transform (Modify) Bone - Bone: dyn_back_02"!(ComponentPose=@n10, bAlphaBoolEnabled:bool=True, Translation="0,0,0", Rotation="-10,0,0", Scale="1,1,1") -> (Pose) {
    alphaInputType=Bool
    rotationMode=BMM_Additive
    rotationSpace=BCS_BoneSpace
    alphaBoolBlend {blendOption=Linear}
    boneToModify {boneName=dyn_back_02}
    rotation {pitch=-10}
   }
   n12 = "Transform (Modify) Bone - Bone: dyn_side_l1_02"!(ComponentPose=@n11, bAlphaBoolEnabled:bool=True, Translation="0,0,0", Rotation="-60,30,0", Scale="1,1,1") -> (Pose) {
    alphaInputType=Bool
    rotationMode=BMM_Additive
    rotationSpace=BCS_BoneSpace
    alphaBoolBlend {blendOption=Linear}
    boneToModify {boneName=dyn_side_l1_02}
    rotation {pitch=-60, yaw=30}
   }
   n13 = "Transform (Modify) Bone - Bone: dyn_side_r2_02"!(ComponentPose=@n12, bAlphaBoolEnabled:bool=True, Translation="0,0,0", Rotation="-59.902,0.539,-0.450", Scale="1,1,1") -> (Pose) {
    alphaInputType=Bool
    rotationMode=BMM_Additive
    rotationSpace=BCS_BoneSpace
    alphaBoolBlend {blendOption=Linear}
    boneToModify {boneName=dyn_side_r2_02}
    rotation=<rot -59.902 0.539 -0.45>
   }
   n14 = Component To Local(ComponentPose=@n13)
   n15 = Control Rig(Source=Use cached pose 'base_pose'()) {
    alphaBoolBlend {blendOption=Linear}
    controlRigAssetReference {blueprintRigClass=&/F_MED_BrutalBurglar_FaceAcc_CtrlRig.F_MED_BrutalBurglar_FaceAcc_CtrlRig_C}
   }
   n16 = RigidBody!(
    ComponentPose=Local To Component(LocalPose=@n15)
    Alpha
    OverrideWorldGravity=Get Gravity Override()
    ExternalForce
    SimSpaceSettings:Sim Space Settings Structure=(WorldAlpha=0.5,VelocityScaleZ=2,DampingAlpha=1,MaxLinearVelocity=500,MaxAngularVelocity=3,MaxLinearAcceleration=1000,MaxAngularAcceleration=1000,ExternalLinearDragV=(X=5,Y=2.3,Z=1),ExternalLinearVelocity=(X=0,Y=0,Z=0),ExternalAngularVelocity=(X=0,Y=0,Z=0))
   ) -> (Pose) {
    bClampLinearTranslationLimitToRefPose=true
    bOverrideWorldGravity=true
    overridePhysicsAsset=&/F_MED_BrutalBurglar_FaceAcc_Physics
    simulationSpace=BaseBoneSpace
    alphaBoolBlend {blendOption=Linear}
    baseBoneRef {boneName=spine_05}
    overrideWorldGravity {z=-980}
    simSpaceSettings {
     maxAngularAcceleration=1000
     maxAngularVelocity=3
     maxLinearAcceleration=1000
     maxLinearVelocity=500
     velocityScaleZ=2
     worldAlpha=0.5
     externalLinearDragV=<5 2.3 1>
    }
   }
   n17 = "Constraint - dyn_back_03 (PP)"!(ComponentPose=@n16, bAlphaBoolEnabled:bool=True, ConstraintWeights_0=0.9, ConstraintWeights_1:float=0.1) -> (Pose) {
    alphaInputType=Bool
    alphaBoolBlend {blendInTime=10, blendOption=QuarticInOut, blendOutTime=5}
    boneToModify {boneName=dyn_back_03}
    constraintSetup [{targetBone {boneName=spine_05}}, {targetBone {boneName=head}}]
    constraintWeights [1,1]
   }
   n18 = "Transform (Modify) Bone - Bone: dyn_back_02"(ComponentPose=@n17, bAlphaBoolEnabled=True, Translation="0,0,0", Rotation="-30,0,0", Scale="1,1,1") {
    alphaInputType=Bool
    rotationMode=BMM_Additive
    rotationSpace=BCS_BoneSpace
    alphaBoolBlend {blendInTime=10, blendOption=QuarticInOut, blendOutTime=5}
    boneToModify {boneName=dyn_back_02}
    rotation {pitch=-30}
   }
   n19 = "Constraint - dyn_back_01 (PP)"!(ComponentPose=@n18, bAlphaBoolEnabled:bool=True, ConstraintWeights_0=1, ConstraintWeights_1:float=0) -> (Pose) {
    alphaInputType=Bool
    alphaBoolBlend {blendInTime=5, blendOption=CubicInOut, blendOutTime=5}
    boneToModify {boneName=dyn_back_01}
    constraintSetup [{targetBone {boneName=spine_05}}, {targetBone {boneName=clavicle_l}}]
    constraintWeights [1,1]
   }
   n20 = "Transform (Modify) Bone - Bone: dyn_back_03"!(ComponentPose=@n19, bAlphaBoolEnabled:bool=True, Translation="0,0,0", Rotation="0,-30,0", Scale="1,1,1") -> (Pose) {
    alphaInputType=Bool
    rotationMode=BMM_Additive
    rotationSpace=BCS_BoneSpace
    alphaBoolBlend {blendInTime=5, blendOption=CubicInOut, blendOutTime=5}
    boneToModify {boneName=dyn_back_03}
    rotation {yaw=-30}
   }
   n21 = "Transform (Modify) Bone - Bone: dyn_fr_strand_l_01"!(ComponentPose=@n20, bAlphaBoolEnabled:bool=True, Translation="0,0,0", Rotation="-10,20,0", Scale="1,1,1") -> (Pose) {
    alphaInputType=Bool
    rotationMode=BMM_Additive
    rotationSpace=BCS_BoneSpace
    alphaBoolBlend {blendInTime=5, blendOption=CubicInOut, blendOutTime=5}
    boneToModify {boneName=dyn_fr_strand_l_01}
    rotation {pitch=-10, yaw=20}
   }
   n22 = "Transform (Modify) Bone - Bone: dyn_back_03"(ComponentPose=@n21, bAlphaBoolEnabled=True, Translation="0,0,0", Rotation="-60,0,0", Scale="1,1,1") {
    alphaInputType=Bool
    rotationMode=BMM_Additive
    rotationSpace=BCS_BoneSpace
    alphaBoolBlend {blendInTime=5, blendOption=CubicInOut, blendOutTime=5}
    boneToModify {boneName=dyn_back_03}
    rotation {pitch=-60}
   }
   n23 = Component To Local(ComponentPose=@n22)
   n24 = Get Backpack_On() -> (Backpack_On:bool)
   n25 = "Constraint - dyn_side_r1_02 (PP)"(
    ComponentPose=Local To Component(LocalPose=Blend Poses by bool(BlendPose_0=@n14, BlendPose_1=@n23, bActiveValue=@n24) {blendType=Linear})
    Alpha=1
    ConstraintWeights_0=0
    ConstraintWeights_1=1
   ) {
    alphaBoolBlend {blendOption=Linear}
    boneToModify {boneName=dyn_side_r1_02}
    constraintSetup [{targetBone {boneName=spine_05}}, {targetBone {boneName=clavicle_r}}]
    constraintWeights [1,1]
   }
   n26 = "Constraint - dyn_side_r2_02 (PP)"(ComponentPose=@n25, Alpha=1, ConstraintWeights_0=0, ConstraintWeights_1=1) {
    alphaBoolBlend {blendOption=Linear}
    boneToModify {boneName=dyn_side_r2_02}
    constraintSetup [{targetBone {boneName=spine_05}}, {targetBone {boneName=clavicle_r}}]
    constraintWeights [1,1]
   }
   n27 = "Constraint - dyn_side_l1_03 (PP)"!(ComponentPose=@n26, Alpha:float=1, ConstraintWeights_0=0.6, ConstraintWeights_1=0.4) -> (Pose) {
    alphaBoolBlend {blendOption=Linear}
    boneToModify {boneName=dyn_side_l1_03}
    constraintSetup [{targetBone {boneName=spine_05}}, {targetBone {boneName=clavicle_l}}]
    constraintWeights [1,1]
   }
   n28 = "Constraint - dyn_side_l2_03 (PP)"(ComponentPose=@n27, Alpha=1, ConstraintWeights_0=0.6, ConstraintWeights_1=0.4) {
    alphaBoolBlend {blendOption=Linear}
    boneToModify {boneName=dyn_side_l2_03}
    constraintSetup [{targetBone {boneName=spine_05}}, {targetBone {boneName=clavicle_l}}]
    constraintWeights [1,1]
   }
   n29 = "Constraint - dyn_back_04 (PPP)"!(ComponentPose=@n28, Alpha=1, ConstraintWeights_0=0, ConstraintWeights_1:float=0.5, ConstraintWeights_2:float=0.5) -> (Pose) {
    alphaBoolBlend {blendOption=Linear}
    boneToModify {boneName=dyn_back_04}
    constraintSetup [{targetBone {boneName=spine_05}}, {targetBone {boneName=clavicle_l}}, {targetBone {boneName=clavicle_r}}]
    constraintWeights [1,1,1]
   }
   n30 = "Transform (Modify) Bone - Bone: dyn_side_l1_03"(ComponentPose=@n29, Alpha=1, Translation="0,0,0", Rotation="0,30,0", Scale="1,1,1") {
    rotationMode=BMM_Additive
    rotationSpace=BCS_BoneSpace
    alphaBoolBlend {blendOption=Linear}
    boneToModify {boneName=dyn_side_l1_03}
    rotation {yaw=30}
   }
   n31 = "Transform (Modify) Bone - Bone: dyn_side_l2_03"(ComponentPose=@n30, Alpha=1, Translation="0,0,0", Rotation="0,30,0", Scale="1,1,1") {
    rotationMode=BMM_Additive
    rotationSpace=BCS_BoneSpace
    alphaBoolBlend {blendOption=Linear}
    boneToModify {boneName=dyn_side_l2_03}
    rotation {yaw=30}
   }
   n32 = "Transform (Modify) Bone - Bone: dyn_side_r1_01"!(ComponentPose=@n31, Alpha:float=1, Translation="0,0,0", Rotation="-30,-10,0", Scale="1,1,1") -> (Pose) {
    rotationMode=BMM_Additive
    rotationSpace=BCS_BoneSpace
    alphaBoolBlend {blendOption=Linear}
    boneToModify {boneName=dyn_side_r1_01}
    rotation {pitch=-30, yaw=-10}
   }
   Output Pose(Result=Component To Local(ComponentPose=@n32)) {layerGroup=None}
   Save cached pose 'base_pose'(Pose=Copy Pose From Mesh(SourceMeshComponent=Get SkelMesh()) {bUseAttachedParent=true}) {cacheName=base_pose}
   Reroute Node(InputPin:Wildcard) -> (OutputPin:Wildcard)
   Reroute Node()
  ]}
 EventGraph {nodes [
   n1 = Event OnCharacterCustomizationLoaded()
   > Set Custom Base Mesh(Custom Base Mesh:Skeletal Mesh Component Object Reference=@n1.CustomBaseMesh) -> (Output_Get:Skeletal Mesh Component Object Reference)
   Event BlueprintUpdateAnimation()
   > n2 = Sequence()
   > n3 = Gravity Override State Function DEPRECATED(
    CharacterPartAnimInstance=Self-Reference()
    GameTimeInSeconds=GetGameTimeInSeconds()
    GravityOverrideParams=Get Gravity Override Params()
    RBAN Params=Get RBAN Params()
   )
   > Set Gravity Override(Gravity Override=@n3)
   > Set RBAN Settings Out(RBAN Settings Out:RBANSetting Struct Structure=@n3.RBAN SettingsOut) -> (Output_Get:RBANSetting Struct Structure)
   n4 = CheckForBackbling()
   Set Backpack_On(execute=@n2.then_1, Backpack_On:bool=OR Boolean!(A=@n4, B=@n4.bIsBackblingACape, C:bool=@n4.bIsBackblingLong) -> (ReturnValue)) -> (Output_Get:bool)
   n5 = Get Custom Base Mesh() -> (Custom Base Mesh:Skeletal Mesh Component Object Reference)
   n6 = GetAlphaFromReferenceTransformDelta(SkeletalMeshComponent=@n5, BoneName=clavicle_r, BaseName=spine_05, DeltaComponentType=EulerY, InputRange=(X=0,Y=40), TwistAxis="1,0,0")
   n7 = GetAlphaFromReferenceTransformDelta(SkeletalMeshComponent=@n5, BoneName=clavicle_l, BaseName=spine_05, DeltaComponentType=EulerY, InputRange=(X=0,Y=40), TwistAxis="1,0,0")
   > Set Hair_on_Emote(Hair_on_Emote:double=SelectFloat(A=float * float(A=@n6, B=@n7) {is=title K2Node_PromotableOperator}, B=0, bPickA=Get bIsPlayingEmote())) -> (Output_Get:double)
   Event BlueprintInitializeAnimation()
   > Set SkelMesh(SkelMesh=GetOwnerSkeletalMeshComponent())
   n8 = Vector Length(A=Get PawnVelocity())
   n9 = "float > float"(A=@n8, B=100)
   n10 = Set Sim Space Settings External Linear Drag V(Sim Space Settings External Linear Drag V:Vector=SelectVector(A="10,2.5,1", B="2.5,2.5,1", bPickA=@n9)) -> (Output_Get:Vector)
   > PrintString!(execute, InString="To String (Boolean)"(InBool=@n9), bPrintToScreen, bPrintToLog, TextColor, Duration, Key) -> (then)
   PrintString(InString=Hello)
   MapRangeClamped(Value=@n8, InRangeA=0, InRangeB=700, OutRangeA=1, OutRangeB=0)
   "To String (Vector)"!(InVec:Vector=@n10) -> (ReturnValue:String)
   MapRangeClamped(Value=Vector Length(A=Get PawnVelocity()), InRangeA=0, InRangeB=700, OutRangeA=1, OutRangeB=0)
   Set Sim Space Settings External Linear Drag V(Sim Space Settings External Linear Drag V="2.5,2.5,1")
  ]}
]
variables [
 Backpack_On {type=boolean}
 Custom Base Mesh {type=SkeletalMeshComponent}
 Gravity Override {type=Vector}
 Gravity Override Params {type=GravityOverrideParamsStruct}
 Hair_on_Emote {type=number}
 RBAN Params {
  type=RBANParamsStruct
  default {
   dBNO {simSpaceSettings {
     maxAngularAcceleration=1000
     maxAngularVelocity=3
     maxLinearAcceleration=300
     maxLinearVelocity=700
     velocityScaleZ=0
     externalLinearDragV=<0 0.5 0>
    }}
   dBNO_Carried {simSpaceSettings {
     maxAngularAcceleration=1000
     maxAngularVelocity=3
     maxLinearAcceleration=300
     maxLinearVelocity=700
     velocityScaleZ=0
     externalLinearDragV=<0 0.5 0>
    }}
   emoteOrMelee {simSpaceSettings {
     maxAngularAcceleration=1000
     maxAngularVelocity=3
     maxLinearAcceleration=300
     maxLinearVelocity=700
     velocityScaleZ=0
     externalLinearDragV=<0 0.5 0>
    }}
   front_End {simSpaceSettings {
     maxAngularAcceleration=200
     maxAngularVelocity=3
     maxLinearAcceleration=200
     maxLinearVelocity=700
     velocityScaleZ=0
     worldAlpha=0.5
     externalLinearDragV=<0 0.5 0>
    }}
   onGround_Crouch_Moving {simSpaceSettings {
     maxAngularAcceleration=1000
     maxAngularVelocity=3
     maxLinearAcceleration=300
     maxLinearVelocity=700
     velocityScaleZ=0
     externalLinearDragV=<0 0.5 0>
    }}
   onGround_Falling {simSpaceSettings {
     maxAngularAcceleration=1000
     maxAngularVelocity=3
     maxLinearAcceleration=300
     maxLinearVelocity=700
     velocityScaleZ=0
     externalLinearDragV=<0 0.5 0>
    }}
   onGround_Moving {
    componentLinearAccScale=<0 0 0>
    componentLinearVelScale=<0 0 0>
    simSpaceSettings {
     maxAngularAcceleration=1000
     maxAngularVelocity=3
     maxLinearAcceleration=1000
     maxLinearVelocity=500
     velocityScaleZ=2
     worldAlpha=0.5
     externalLinearDragV=<2.5 2.5 2.5>
    }
   }
   onGround_Standing {
    componentLinearAccScale=<0 0 0>
    componentLinearVelScale=<0 0 0>
    simSpaceSettings {
     maxAngularAcceleration=1000
     maxAngularVelocity=3
     maxLinearAcceleration=1000
     maxLinearVelocity=500
     velocityScaleZ=2
     worldAlpha=0.5
     externalLinearDragV=<4.8 4.8 4.8>
    }
   }
   rail_Grinding {simSpaceSettings {
     maxAngularAcceleration=1000
     maxAngularVelocity=3
     maxLinearAcceleration=300
     maxLinearVelocity=700
     velocityScaleZ=0
     worldAlpha=0.1
     externalLinearDragV=<0 0.5 0>
    }}
   riding_Moving {simSpaceSettings {
     maxAngularAcceleration=1000
     maxAngularVelocity=3
     maxLinearAcceleration=300
     maxLinearVelocity=700
     velocityScaleZ=0
     externalLinearDragV=<0 0.5 0>
    }}
   riding_Moving_Targeting {simSpaceSettings {
     maxAngularAcceleration=1000
     maxAngularVelocity=3
     maxLinearAcceleration=300
     maxLinearVelocity=700
     velocityScaleZ=0
     externalLinearDragV=<0 0.5 0>
    }}
   riding_Standing {simSpaceSettings {
     maxAngularAcceleration=1000
     maxAngularVelocity=3
     maxLinearAcceleration=300
     maxLinearVelocity=700
     velocityScaleZ=0
     externalLinearDragV=<0 0.5 0>
    }}
   skydive_DivingDown {simSpaceSettings {
     maxAngularAcceleration=1000
     maxAngularVelocity=3
     maxLinearAcceleration=300
     maxLinearVelocity=700
     velocityScaleZ=0
     externalLinearDragV=<0 0.5 0>
    }}
   skydive_FloatingDown {
    componentLinearAccScale=<0 0.5 0>
    componentLinearVelScale=<0 0.5 0>
    simSpaceSettings {
     maxAngularAcceleration=1000
     maxAngularVelocity=3
     maxLinearAcceleration=300
     maxLinearVelocity=200
     velocityScaleZ=0
     externalLinearDragV=<0 0.5 0>
    }
   }
   skydive_GoingUp {simSpaceSettings {
     maxAngularAcceleration=1000
     maxAngularVelocity=3
     maxLinearAcceleration=300
     maxLinearVelocity=700
     velocityScaleZ=0
     externalLinearDragV=<0 0.5 0>
    }}
   skydive_Parachuting {
    componentLinearAccScale=<0 0.1 0.3>
    componentLinearVelScale=<0 0.1 0.3>
    simSpaceSettings {
     maxAngularAcceleration=1000
     maxAngularVelocity=3
     maxLinearAcceleration=300
     maxLinearVelocity=700
     velocityScaleZ=0
     externalLinearDragV=<0 0.5 0>
    }
   }
   skydive_Surfing {simSpaceSettings {
     maxAngularAcceleration=1000
     maxAngularVelocity=3
     maxLinearAcceleration=300
     maxLinearVelocity=700
     velocityScaleZ=0
     externalLinearDragV=<0 0.5 0>
    }}
   swimming {simSpaceSettings {
     maxAngularAcceleration=1000
     maxAngularVelocity=3
     maxLinearAcceleration=300
     maxLinearVelocity=700
     velocityScaleZ=0
     externalLinearDragV=<0 0.5 0>
    }}
   tactical_Sprint {simSpaceSettings {
     maxAngularAcceleration=1000
     maxAngularVelocity=3
     maxLinearAcceleration=300
     maxLinearVelocity=700
     velocityScaleZ=0
     externalLinearDragV=<0 0.5 0>
    }}
   ziplining {simSpaceSettings {
     maxAngularAcceleration=1000
     maxAngularVelocity=3
     maxLinearAcceleration=300
     maxLinearVelocity=700
     velocityScaleZ=0
     externalLinearDragV=<0 0.5 0>
    }}
  }
 }
 RBAN Settings Out {
  type=RBANSettingStruct
  default {
   componentLinearAccScale=<0 0 0>
   componentLinearVelScale=<0 0 0>
   simSpaceSettings {
    dampingAlpha=1
    maxAngularAcceleration=1000
    maxAngularVelocity=3
    maxLinearAcceleration=300
    maxLinearVelocity=700
    velocityScaleZ=0
    worldAlpha=1
    externalAngularVelocity=<0 0 0>
    externalLinearDragV=<0 0.5 0>
    externalLinearVelocity=<0 0 0>
   }
  }
 }
 Sim Space Settings External Linear Drag V {type=Vector, default=<5 2.5 1>}
 SkelMesh {type=SkeletalMeshComponent}
]
